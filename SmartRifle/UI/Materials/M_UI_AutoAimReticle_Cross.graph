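# Translucent UI material that draws the Smart Rifle's auto-aim reticle as radial prongs, four by default, with a dot.
# A reticle UV function repeats and rotates a box SDF, offsetting it by remapped accuracy spread, while a circle SDF forms the dot.
# Anti-aliased shapes combine through Max, with drop shadows blended from thresholded copies of the same SDFs.
# CrosshairOpacity fades prongs independently of the dot, and MasterOpacity scales the combined shadowed mask.
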Material /SmartRifle/UI/Materials/M_UI_AutoAimReticle_Cross
graph {
 properties {bAutomaticallySetUsageInEditor=false, blendMode=BLEND_Translucent, materialDomain=MD_UI, usedShadingModels=Default Lit}
 nodes [
  n1 = MaterialFunctionCall(
   UVs=NamedRerouteUsage()
   Size=AppendVector(
    A=ScalarParameter() {defaultValue=0.075, group=Shape Operators, parameterName=BoxWidth, sliderMax=1}
    B=ScalarParameter() {defaultValue=0.4, group=Shape Operators, parameterName=BoxHeight, sliderMax=1}
   )
  ) {materialFunction=&MF_UI_Box}
  n2 = MaterialFunctionCall(DistanceField=@n1.SDF) {materialFunction=&MF_UI_AntiAliasing}
  n3 = NamedRerouteUsage()
  n4 = MaterialFunctionCall(DistanceField=@n3) {materialFunction=&MF_UI_AntiAliasing}
  Max(A=@n2.Inner, B=@n4.Inner)
  n5 = ScalarParameter() {defaultValue=0.03, desc=Controls the thickness of the Drop Shadow, group=Drop Shadow Operators, parameterName=ShadowThickness, sliderMax=0.1}
  n6 = ScalarParameter() {defaultValue=0.4, desc=Controls the opacity of the drop shadow., group=Drop Shadow Operators, parameterName=ShadowOpacity, sliderMax=1}
  n7 = LinearInterpolate(
   A=@n2.Inner
   B=MaterialFunctionCall(DistanceField=@n1.SDF, Threshold=@n5) {materialFunction=&MF_UI_AntiAliasing}.Inner
   Alpha=@n6
  ) {constAlpha=0.4}
  n8 = ScalarParameter() {defaultValue=1, parameterName=CrosshairOpacity}
  n9 = MaterialFunctionCall(DistanceField=@n3, Threshold=Constant() {desc=Controls the thickness of the Drop Shadow, r=0.015}) {materialFunction=&MF_UI_AntiAliasing}
  Multiply(
   A=Max(A=Multiply(A=@n7, B=@n8) {constB=0}, B=LinearInterpolate(A=@n4.Inner, B=@n9.Inner, Alpha=@n6) {constAlpha=0.4})
   B=ScalarParameter() {defaultValue=0.8, desc=Controls Overall Opacity, group=Reticle Operators, parameterName=MasterOpacity, sliderMax=1}
  )
  n10 = ScalarParameter() {defaultValue=25.6, desc=Variable for Accuracy Spread Data, group=Reticle Operators, parameterName=SpreadOffset, sliderMax=100}
  n11 = ScalarParameter() {desc=Variable to Position Reticle Prongs., group=Reticle Operators, parameterName=PositionalOffset, sliderMax=1, sliderMin=-1}
  NamedRerouteDeclaration(None=MaterialFunctionCall!(
    Reticle UVs
    Reticle Sides=ScalarParameter() {defaultValue=4, group=Reticle Operators, parameterName=ReticleSides, sliderMax=12}
    Reticle Scale
    Reticle Rotation=ScalarParameter() {defaultValue=0.125, group=Reticle Operators, parameterName=ReticleRotation, sliderMax=1}
    Prong Scale=ScalarParameter() {defaultValue=1, group=Prong Operators, parameterName=Prong Scale, sliderMax=2}
    Prong Rotation=ScalarParameter() {group=Prong Operators, parameterName=ProngRotation, sliderMax=1}
    Spread Offset=@n10
    Positional Offset=@n11
    "Spread Input Range [Min,Max]"=Constant2Vector() {g=40, r=15}
    "Spread Remapped Range [Min,Max]"=Constant2Vector() {g=0.2, r=-0.1}
   ) {materialFunction=/FortUILibrary/Systems/Reticles/MaterialFunctions/MF_UI_Reticle_UVs}.UVs) {name=Uvs, nodeColor=<rgba 0.025 0.0203 0.0125 0.75>}
  NamedRerouteDeclaration(None=MaterialFunctionCall!(UVs, Radius=ScalarParameter() {defaultValue=0.05, group=Dot Modifiers, parameterName=DotSize, sliderMax=0.4, sliderMin=0.16, sortPriority=1}) {materialFunction=/FortUILibrary/MaterialFunctions/Shapes/SDFs/MF_SDF_Circle}.SDF) {name=Dot, nodeColor=<rgba 0 0 0 1>}
  ScalarParameter() {defaultValue=1, group=Reticle Operators, parameterName=ReticleScale, sliderMax=2}
 ]
}
