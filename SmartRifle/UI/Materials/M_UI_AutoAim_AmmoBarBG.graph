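# Translucent UI material for the Smart Rifle auto-aim ammo bar background, judging by its name and path.
# It feeds rotated 9-slice UVs into an eight-sided polygon SDF, which an anti-aliasing function turns into a filled inner mask.
# That mask is multiplied by a dimmed, inverted reflected gradient, then scaled by an Opacity parameter, so the shape is procedural and uses no texture samples in this graph.
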Material /SmartRifle/UI/Materials/M_UI_AutoAim_AmmoBarBG
graph {
 properties {bAutomaticallySetUsageInEditor=false, blendMode=BLEND_Translucent, materialDomain=MD_UI, usedShadingModels=Default Lit}
 nodes [
  Constant() {r=1}
  n1 = MaterialFunctionCall!(
   UVs=MaterialFunctionCall(
    UVs=MaterialFunctionCall!(UVs) {materialFunction=/FortUILibrary/MaterialFunctions/Basics/MF_UI_9SliceUVs}.UVs
    "Rotation Angle (0-1)"=Constant() {r=0.313}
   ) {materialFunction=&CustomRotator}.Rotated Values
   Size=Constant() {r=0.9}
   Sides=Constant() {r=8}
   InnerAngle
   OutlineThickness
   OutlineAlignment
   GlowMin
   GlowMax
  ) {materialFunction=/FortUILibrary/MaterialFunctions/Shapes/MF_UI_RegularPolygon}
  n2 = OneMinus(None=MaterialFunctionCall!(UVs, Offset, RotationCenter, RotationAngle) {materialFunction=/FortUILibrary/MaterialFunctions/Gradients/MF_UI_ReflectedGradient}.Result)
  Multiply(
   A=Multiply(
    A=MaterialFunctionCall(DistanceField=@n1.SDF) {materialFunction=&MF_UI_AntiAliasing}.Inner
    B=Multiply(A=@n2) {constB=0.4}
   )
   B=ScalarParameter() {defaultValue=0.4, parameterName=Opacity}
  )
 ]
}
