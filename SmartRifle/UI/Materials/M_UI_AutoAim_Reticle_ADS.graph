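# Translucent UI material for the Smart Rifle auto-aim reticle shown when aiming down sights.
# Builds the reticle procedurally from trapezoid and box SDFs combined by boolean operations, then anti-aliased at several thresholds for fill, outline and glows.
# Anim morphs the shape, while Progress drives a radial fill and samples a curve atlas for glow, middle-piece size and vertical shift.
# TargetLocked adjusts layer intensity and Color supplies a tint.
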Material /SmartRifle/UI/Materials/M_UI_AutoAim_Reticle_ADS
graph {
 properties {bAutomaticallySetUsageInEditor=false, blendMode=BLEND_Translucent, materialDomain=MD_UI, usedShadingModels=Default Lit}
 nodes [
  Constant3Vector() {constant=<rgba 1 0.2874 0.0331 1>}
  Constant() {r=3.5}
  Constant() {r=0.03}
  n1 = LinearInterpolate(A=NamedRerouteUsage(), B=NamedRerouteUsage(), Alpha=Constant() {desc=OutlineOpacity, r=0.8})
  n2 = Multiply(A=@n1) {constB=0.5}
  n3 = Multiply(A=@n1) {constB=0.8}
  n4 = ScalarParameter() {defaultValue=1, parameterName=TargetLocked}
  LinearInterpolate(
   A=LinearInterpolate(A=LinearInterpolate(A=@n2, B=@n3, Alpha=@n4), B=@n1, Alpha=NamedRerouteUsage())
   B=LinearInterpolate(A=NamedRerouteUsage(), B=NamedRerouteUsage(), Alpha=Constant() {desc=GlowOpacity, r=1})
   Alpha=CurveAtlasRowParameter(CurveTime=ScalarParameter() {parameterName=Progress}) {atlas=&/CurveAtlas_AutorAim, curve=&/Curve_AutoAim, parameterName=GlowCurve}.G
  )
  n5 = Constant3Vector() {constant=<rgba 0 0 0 1>}
  n6 = ScalarParameter() {parameterName=Progress}
  n7 = Constant() {r=1}
  n8 = LinearInterpolate(A=Constant() {r=1}, B=VectorParameter() {parameterName=Color, defaultValue=<rgba 1 1 1 1>}.RGB, Alpha=NamedRerouteUsage())
  n9 = CurveAtlasRowParameter(CurveTime=ScalarParameter() {defaultValue=1, parameterName=Anim}) {atlas=&CurveAtlas_AutorAim, curve=&Curve_AutoAim, parameterName=GlowCurve}
  n10 = Constant3Vector() {constant=<rgba 1 0.0509 0 1>}
  n11 = Constant3Vector() {constant=<rgba 1 0.5457 0.1559 1>}
  n12 = NamedRerouteUsage()
  n13 = ScalarParameter() {parameterName=Progress}
  LinearInterpolate(
   A=LinearInterpolate(
    A=Constant()
    B=@n5
    Alpha=CurveAtlasRowParameter(CurveTime=@n6) {atlas=&CurveAtlas_AutorAim, curve=&Curve_AutoAim, parameterName=GlowCurve}.G
   )
   B=LinearInterpolate(
    A=LinearInterpolate(A=@n7, B=@n8, Alpha=@n9.G) {constA=1, constB=0}
    B=LinearInterpolate(A=@n10, B=@n11, Alpha=@n12)
    Alpha=CurveAtlasRowParameter(CurveTime=@n13) {atlas=&CurveAtlas_AutorAim, curve=&Curve_AutoAim, parameterName=GlowCurve}.G
   )
   Alpha=NamedRerouteUsage()
  )
  LinearInterpolate(A=@n5, B=Constant3Vector() {constant=<rgba 0.8854 0.1356 0 1>}, Alpha=NamedRerouteUsage()) {constB=0}
  LinearInterpolate() {constB=0.17}
  n14 = CurveAtlasRowParameter(CurveTime=ScalarParameter() {parameterName=Progress}) {atlas=&CurveAtlas_AutorAim, curve=&Curve_AutoAim, parameterName=GlowCurve}
  MaterialFunctionCall(
   UVs=Add(A=NamedRerouteUsage(), B=Constant2Vector() {g=0.08}) {constB=0}
   Size=AppendVector(A=Constant() {r=0.21}, B=LinearInterpolate(Alpha=@n14) {constB=0.17})
  ) {materialFunction=&MF_UI_Box}
  n15 = ScalarParameter() {defaultValue=1, parameterName=Anim}
  n16 = Add(A=NamedRerouteUsage(), B=AppendVector(A=Constant() {r=0.5}, B=LinearInterpolate(Alpha=@n15) {constA=0.52, constB=0.48}))
  n17 = Constant2Vector() {g=1, r=0.5}
  MaterialFunctionCall(
   UVs=MaterialFunctionCall(UVs=Multiply(A=@n16, B=@n17), "Rotation Angle (0-1)"=Constant() {r=0.125}) {materialFunction=&CustomRotator}.Rotated Values
   Size=Constant() {r=1}
  ) {materialFunction=&MF_UI_Box}
  n18 = NamedRerouteUsage()
  n19 = NamedRerouteUsage()
  n20 = MaterialFunctionCall!(
   DistanceFieldA=MaterialFunctionCall!(
    DistanceFieldA=MaterialFunctionCall!(DistanceFieldA=@n18, DistanceFieldB=@n19, BlendAmount) {materialFunction=/FortUILibrary/MaterialFunctions/Shapes/Modifiers/MF_UI_SDF_Union}.Exact
    DistanceFieldB=NamedRerouteUsage()
    BlendAmount
   ) {materialFunction=/FortUILibrary/MaterialFunctions/Shapes/Modifiers/MF_UI_SDF_Subtraction}.Exact
   DistanceFieldB=NamedRerouteUsage()
   BlendAmount
  ) {materialFunction=/FortUILibrary/MaterialFunctions/Shapes/Modifiers/MF_UI_SDF_Intersection}
  NamedRerouteDeclaration(None=MaterialFunctionCall(DistanceField=Subtract(A=@n20.Exact) {constB=-0.045}) {materialFunction=&MF_UI_AntiAliasing}.Inner) {name=InnerGlow, nodeColor=<rgba 1 0.4219 0 1>}
  n21 = NamedRerouteUsage()
  n22 = Constant2Vector() {g=-0.13}
  n23 = Constant() {r=0.07}
  n24 = ScalarParameter() {defaultValue=1, parameterName=Anim}
  n25 = LinearInterpolate(Alpha=@n24) {constB=0.11}
  n26 = LinearInterpolate(Alpha=@n24) {constB=0.5}
  n27 = Constant() {r=0.89}
  NamedRerouteDeclaration(None=MaterialFunctionCall!(
    UVs=Add(A=@n21, B=@n22) {constB=0}
    TopWidth=Add(A=@n23, B=@n25)
    BottomWidth=Add(A=@n26, B=@n27)
    Height=Constant() {r=0.6}
    OutlineThickness=Constant() {r=0.0935}
    OutlineAlignment=Constant() {r=-0.8}
    GlowMin
    GlowMax
   ) {materialFunction=/FortUILibrary/MaterialFunctions/Shapes/MF_UI_IsoscelesTrapezoid}.SDFOutline) {name=TrapezoidOutline, nodeColor=<rgba 0 0 0 1>}
  n28 = NamedRerouteUsage()
  n29 = Constant2Vector() {g=0.095}
  n30 = CurveAtlasRowParameter(CurveTime=ScalarParameter() {parameterName=Progress}) {atlas=&CurveAtlas_AutorAim, curve=&Curve_AutoAim, parameterName=GlowCurve}
  NamedRerouteDeclaration(None=MaterialFunctionCall(
    UVs=Add(A=@n28, B=@n29) {constB=0}
    TopWidth=Constant() {r=0.21}
    BottomWidth=LinearInterpolate(Alpha=@n30.G) {constA=0.2, constB=0.5}
    Height=LinearInterpolate(Alpha=@n30.G) {constB=0.14}
   ) {materialFunction=&MF_UI_IsoscelesTrapezoid}.SDF) {name=MiddleBoxHeight, nodeColor=<rgba 0 0 0 1>}
  n31 = LinearInterpolate(Alpha=ScalarParameter() {defaultValue=1, parameterName=Anim}) {constA=0.09, constB=-0.05}
  n32 = Constant() {r=1}
  NamedRerouteDeclaration(None=MaterialFunctionCall(Size=AppendVector(A=@n31, B=@n32)) {materialFunction=&MF_UI_Box}.SDF) {name=MiddleCut, nodeColor=<rgba 0 0 0 1>}
  n33 = NamedRerouteUsage()
  n34 = ScalarParameter() {defaultValue=1, parameterName=Anim}
  n35 = AppendVector(A=Constant(), B=LinearInterpolate(Alpha=@n34) {constA=0.265, constB=0.13})
  NamedRerouteDeclaration(None=MaterialFunctionCall(
    UVs=Add(A=@n33, B=@n35) {constB=0}
    TopWidth=LinearInterpolate(Alpha=@n34) {constA=3.5, constB=1.9}
    BottomWidth=Constant() {r=0.22}
    Height=Constant() {r=0.8}
   ) {materialFunction=&MF_UI_IsoscelesTrapezoid}.SDF) {name=LowerMask, nodeColor=<rgba 0 0 0 1>}
  NamedRerouteDeclaration(None=MaterialFunctionCall(DistanceField=@n20.Exact) {materialFunction=&MF_UI_AntiAliasing}.Inner) {name=InnerShape, nodeColor=<rgba 0 0 0 1>}
  NamedRerouteDeclaration(None=MaterialFunctionCall(DistanceField=@n20.Exact, Threshold=Constant() {r=0.05}) {materialFunction=&MF_UI_AntiAliasing}.Inner) {name=Outline, nodeColor=<rgba 0 0 0 1>}
  NamedRerouteDeclaration(None=MaterialFunctionCall(DistanceField=@n20.Exact, Threshold=Constant() {r=0.1}) {materialFunction=&MF_UI_AntiAliasing}.Inner) {name=OuterGlow, nodeColor=<rgba 0 0 0 1>}
  n36 = MaterialFunctionCall(UVs=Add(A=NamedRerouteUsage(), B=Constant2Vector() {g=-0.21}), "Rotation Angle (0-1)"=Constant() {r=0.21}) {materialFunction=&CustomRotator}
  n37 = Constant()
  n38 = ScalarParameter() {parameterName=Progress}
  NamedRerouteDeclaration(None=Step(
    Y=MaterialFunctionCall!(UV=@n36.Rotated Values, "Arc Start Angle (0-1)", "Arc End Angle (0-1)", "Fill Ratio (0-1)", Fill Clockwise=@n37) {materialFunction=/FortUILibrary/MaterialFunctions/Basics/MF_UI_RadialArc}.Radial Fill
    X=LinearInterpolate(Alpha=@n38) {constB=0.42}
   )) {name=Progress, nodeColor=<rgba 0 0 0 1>}
  n39 = Constant()
  n40 = LinearInterpolate(Alpha=CurveAtlasRowParameter(CurveTime=ScalarParameter() {parameterName=Progress}) {atlas=&CurveAtlas_AutorAim, curve=&Curve_AutoAim, parameterName=GlowCurve}.R) {constA=-0.1, constB=0.1}
  NamedRerouteDeclaration(None=Add(A=MaterialFunctionCall() {materialFunction=&GetUserInterfaceUV}.Normalized UV, B=AppendVector(A=@n39, B=@n40))) {name=UV, nodeColor=<rgba 0 0 0 1>}
  NamedRerouteUsage()
  Saturate(None=ScalarParameter() {desc=ThicknessOpacity, parameterName=Glow})
  ScalarParameter() {defaultValue=1, parameterName=TargetLocked}
  ScalarParameter() {defaultValue=1, parameterName=Anim}
  ScalarParameter() {parameterName=Progress}
 ]
}
