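# Translucent UI material that procedurally draws a Smart Rifle auto-aim "targeted" marker, with no textures.
# Scaled UVs are rotated, mirrored, offset and rotated again to shape a box outline, trimmed by SDF subtraction of another box.
# A Min unions this with a small box.
# Anti-aliased thresholds on the combined distance field layer a dark base, red fill and pale orange core, plus a wider mask.
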Material /SmartRifle/UI/Materials/M_UI_AutoAim_Targeted
graph {
 properties {bAutomaticallySetUsageInEditor=false, blendMode=BLEND_Translucent, materialDomain=MD_UI, usedShadingModels=Default Lit}
 nodes [
  n1 = MaterialFunctionCall(
   UVs=MaterialFunctionCall() {materialFunction=&GetUserInterfaceUV}.Normalized UV
   Texture Scale=Constant() {r=1.5}
  ) {materialFunction=&ScaleUVsByCenter}
  n2 = MaterialFunctionCall!(
   UV=MaterialFunctionCall(UVs=NamedRerouteDeclaration(None=@n1.UVs) {name=UV, nodeColor=<rgba 0 0 0 1>}, "Rotation Angle (0-1)"=Constant() {r=1.25}) {materialFunction=&CustomRotator}.Rotated Values
   Mirror Y
   Swizzle UV
  ) {materialFunction=/FortUILibrary/MaterialFunctions/Basics/MF_UI_Symmetry_2x1_New}
  n3 = Constant2Vector() {r=-0.4}
  n4 = MaterialFunctionCall(
   UVs=MaterialFunctionCall(UVs=Add(A=@n2.Symmetry, B=@n3), Texture Scale=Constant2Vector() {g=0.3, r=1}) {materialFunction=&ScaleUVsByCenter}.UVs
   "Rotation Angle (0-1)"=Constant() {r=0.125}
  ) {materialFunction=&CustomRotator}
  n5 = Constant() {r=0.1}
  n6 = MaterialFunctionCall(UVs=NamedRerouteUsage(), Size=Constant2Vector() {g=0.14, r=1}) {materialFunction=&MF_UI_Box}
  n7 = NamedRerouteUsage()
  n8 = Constant2Vector() {g=0.1, r=0.04}
  n9 = Min(
   A=MaterialFunctionCall!(
    DistanceFieldA=MaterialFunctionCall(UVs=@n4.Rotated Values, OutlineThickness=@n5) {materialFunction=&MF_UI_Box}.SDFOutline
    DistanceFieldB=Multiply(A=@n6.SDF) {constB=2.5, desc=OutlineThickness}
    BlendAmount
   ) {materialFunction=/FortUILibrary/MaterialFunctions/Shapes/Modifiers/MF_UI_SDF_Subtraction}.Exact
   B=Multiply(A=MaterialFunctionCall(UVs=@n7, Size=@n8) {materialFunction=&MF_UI_Box}.SDF) {constB=3, desc=OutlineThickness}
  )
  LinearInterpolate(
   A=LinearInterpolate(
    B=Constant3Vector() {constant=<rgba 1 0.0513 0 1>}
    Alpha=MaterialFunctionCall(DistanceField=@n9) {materialFunction=&MF_UI_AntiAliasing}.Inner
   )
   B=Constant3Vector() {constant=<rgba 1 0.5457 0.1559 1>}
   Alpha=MaterialFunctionCall(DistanceField=@n9, Threshold=Constant() {r=-0.05}) {materialFunction=&MF_UI_AntiAliasing}.Inner
  )
  MaterialFunctionCall(DistanceField=@n9, Threshold=Constant() {r=0.08}) {materialFunction=&MF_UI_AntiAliasing}
 ]
}
